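# Unlit translucent Niagara sprite material that draws a procedural impact ring for the Lettuce pickaxe.
# Subtracts an inner SphereMask from an outer one, with dynamic parameters driving inverted radius, width and horizontal inner-center offset for crescent shapes.
# Particle alpha scales the mask into a DepthFade, while particle color feeds MF_FX_EmissiveAdjustments.
# Scalar parameters Center Offset X, Width and Impact Control are unconnected.
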Material /BRCosmetics/Weapons/FORT_Melee/Pickaxe_Lettuce/Materials/M_ImpactCircle
graph {
 properties {bUsedWithNiagaraSprites=true, blendMode=BLEND_Translucent, pixelDepthOffsetMode=PDOM_Legacy, shadingModel=MSM_Unlit, usedShadingModels=Unlit}
 nodes [
  n1 = ParticleColor()
  n2 = TextureCoordinate()
  n3 = Constant2Vector() {g=0.5, r=0.5}
  n4 = DynamicParameter() {paramNames [Center Offset X, Width, Radius, Param4]}
  n5 = Divide(A=OneMinus(None=@n4.Radius))
  n6 = ScalarParameter() {parameterName=Hardness, sliderMax=1}
  n7 = SphereMask(A=@n2, B=@n3, Radius=@n5, Hardness=@n6) {hardnessPercent=50}
  n8 = Constant()
  n9 = SphereMask(A=@n2, B=Subtract(A=@n3, B=AppendVector(A=@n4.Center Offset X, B=@n8)), Radius=Multiply(A=@n5, B=OneMinus(None=@n4.Width)), Hardness=@n6) {hardnessPercent=50}
  DepthFade(
   Opacity=Multiply(A=@n1.A, B=Subtract(A=@n7, B=@n9))
   FadeDistance=ScalarParameter() {defaultValue=30, parameterName=Depth Fade Distance, sliderMax=100}
  )
  MaterialFunctionCall(Emissive=@n1.RGB) {materialFunction=&MF_FX_EmissiveAdjustments}
  ScalarParameter() {defaultValue=0.05, parameterName=Center Offset X, sliderMax=0.5, sliderMin=-0.5}
  ScalarParameter() {parameterName=Impact Control, sliderMax=1}
  ScalarParameter() {defaultValue=0.1, parameterName=Width}
 ]
}
